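# Social UI building-block widget that draws a configurable backplate for social buttons, using one image with the M_UI_Social_Backplate material in a grid panel.
# Functions set dynamic material parameters for color variant, corner radius and secondary gradient color, the last through an MID cached in PreConstruct.
# The color enum maps to a switch scalar, with one variant swapping value under touch input.
# ContainBackplate toggles padding on the image's grid slot.
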WidgetBlueprint /FortniteCoreUI/Social/Buttons/Blocks/WBP_Social_Block_Backplate
graphs [
 ContainBackplate {nodes [
   n1 = ContainBackplate!() -> (then, ContainBackplatePadding:bool=~) {is=title K2Node_FunctionEntry}
   > Set Padding!(
    execute
    Padding:Margin Structure=Select!(
     Option 0_Left:float=0
     Option 0_Top:float=0
     Option 0_Right:float=0
     Option 0_Bottom:float=0
     Option 1_Left:float=16
     Option 1_Top:float=16
     Option 1_Right:float=16
     Option 1_Bottom:float=16
     Index=@n1
    ) -> (ReturnValue:Margin Structure)
    self:Grid Slot Object Reference=SlotAsGridSlot(Widget=Get Backplate() -> (Backplate:Image Object Reference))
   ) -> (then, Output_Get:Margin Structure)
  ]}
 EventGraph {nodes [
   Event PreConstruct()
   n1 = Get Backplate()
   > Set BackplateMID(BackplateMID:Material Instance Dynamic Object Reference=GetDynamicMaterial!(Brush:"Slate Brush Structure (by ref)"=Get Brush!(self:Image Object Reference=@n1) -> (Brush:Slate Brush Structure)) -> (ReturnValue)) -> (Output_Get:Material Instance Dynamic Object Reference)
   Event Construct()
  ]}
 SetBackplateColor {nodes [
   n1 = SetBackplateColor!() -> (then, Backplate Switch:E_UI_Social_Button_Backplate_Color Enum=NewEnumerator0) {is=title K2Node_FunctionEntry}
   n2 = Get Backplate()
   n3 = IsUsingInputType!(InputType:ECommonInputType Enum=Touch, PlayerController:Player Controller Object Reference=GetOwningPlayer()) -> (ReturnValue:bool)
   > SetScalarParameterValue(
    self=GetDynamicMaterial(Brush=Get Brush(self=@n2))
    ParameterName=Backplate Switch
    Value=Select!(
     NewEnumerator0:float=0
     NewEnumerator1:float=1
     NewEnumerator2:float=2
     NewEnumerator3:float=3
     NewEnumerator4:float=SelectFloat(A=1, B=4, bPickA=@n3)
     NewEnumerator5:float=5
     NewEnumerator6:float=6
     NewEnumerator7:float=7
     NewEnumerator8:float=8
     NewEnumerator9:float=9
     NewEnumerator11:float=10
     NewEnumerator12:float=11
     NewEnumerator13:float=12
     Index:E_UI_Social_Button_Backplate_Color Enum=@n1
    ) -> (ReturnValue:float)
   )
  ]}
 SetCornerRadius {nodes [
   SetCornerRadius!() -> (then, Radius In Pixels:bool=~, Corner Radius:double=~) {is=title K2Node_FunctionEntry}
   n1 = Get Backplate()
   > SetScalarParameterValue(
    self=GetDynamicMaterial(Brush=Get Brush(self=@n1))
    ParameterName=Radius In Pixels
    Value="To Float (Boolean)"(InBool=Get Radius In Pixels() -> (Radius In Pixels:bool))
   )
   n2 = Get Backplate()
   > SetScalarParameterValue(self=GetDynamicMaterial(Brush=Get Brush(self=@n2)), ParameterName=Corner Radius, Value=Get Corner Radius() -> (Corner Radius:double))
  ]}
 SetSecondaryColor {nodes [
   n1 = SetSecondaryColor!() -> (then, Secondary Color:Linear Color Structure) {is=title K2Node_FunctionEntry}
   > SetVectorParameterValue(self=Get BackplateMID() -> (BackplateMID:Material Instance Dynamic Object Reference), ParameterName=Secondary Gradient Color, Value=@n1)
  ]}
]
variables [BackplateMID {type=MaterialInstanceDynamic}]
widgets {
 info {inheritedWidgetCount=0, parentClass=&UserWidget, rootWidgetClass=&GridPanel, widgetCount=2}
 widgets [
  Root {widgetClassPath=&GridPanel, widget=Root, properties {columnFill [1], rowFill [1]}}
  Backplate {
   widgetClassPath=&Image
   bIsVariable=true
   parent=Root
   widget=Backplate
   properties {brush {
     bIsDynamicallyLoaded=false
     drawAs=Image
     imageType=FullColor
     mirroring=NoMirror
     resourceName=None
     resourceObject=/FortniteCoreUI/Social/Buttons/Materials/M_UI_Social_Backplate
     tiling=NoTile
     imageSize=<32 32>
     margin=<ltrb 0 0 0 0>
     outlineSettings {
      bUseBrushTransparency=false
      roundingType=HalfHeightRadius
      width=0
      color {colorUseRule=UseColor_Specified, specifiedColor=<rgba 0 0 0 0>}
      cornerRadii=<0 0 0 0>
     }
     tintColor {colorUseRule=UseColor_Specified, specifiedColor=<rgba 1 1 1 1>}
     uVRegion {bIsValid=false, max=<0 0>, min=<0 0>}
    }}
   slot {class=GridSlot}
  }
 ]
}
